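# Animation Blueprint for the Llama Roaster weapon, animating its spinning barrel and overheat-driven ear flaps.
# Each update reads the owning weapon, stepping barrel roll up with wrap while time-to-next-fire is positive, otherwise back toward zero.
# Negated overheat value becomes a yaw added to three ear bones, while barrel rotation replaces the main barrel bone's.
# The firing overlay sequence and the IsOverheated setter are left unconnected.
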AnimBlueprint /MotherGameplay/Items/LlamaRoaster/Animation/Llama_Roaster_AnimBP
target_skeleton=/MotherGameplay/Items/LlamaRoaster/Assets/Meshes/Llama_Roaster_Skeleton
graphs [
 AnimGraph {nodes [
   Local Space Ref Pose()
   Local Space Ref Pose()
   n1 = Apply Additive(Base=Sequence Player 'LlamaRoaster_Base_Rifle'() {sequence=&/LlamaRoaster_Base_Rifle}) {alphaBoolBlend {blendOption=Linear}}
   n2 = Sequence Player 'LlamaRoaster_Base_Rifle'() {sequence=&LlamaRoaster_Base_Rifle}
   n3 = Get IsFiring() -> (IsFiring:bool)
   n4 = "Transform (Modify) Bone - Bone: mainBarrel_base"(
    ComponentPose=Local To Component(LocalPose=Blend Poses by bool(BlendPose_0=@n1, BlendPose_1=@n2, BlendTime_1=0.25, bActiveValue=@n3) {blendType=Linear})
    Alpha=1
    Translation="0,0,0"
    Rotation=Get BarrelROT() -> (BarrelROT:Rotator)
    Scale="1,1,1"
   ) {rotationMode=BMM_Replace, rotationSpace=BCS_ParentBoneSpace, alphaBoolBlend {blendOption=Linear}, boneToModify {boneName=mainBarrel_base}}
   n5 = Get Overheat_FlapRot() -> (Overheat_FlapRot:Rotator)
   n6 = "Transform (Modify) Bone - Bone: Mid_ear_jnt"(ComponentPose=@n4, Alpha=1, Translation="0,0,0", Rotation=@n5, Scale="1,1,1") {rotationMode=BMM_Additive, rotationSpace=BCS_BoneSpace, alphaBoolBlend {blendOption=Linear}, boneToModify {boneName=Mid_ear_jnt}}
   n7 = "Transform (Modify) Bone - Bone: L_ear_jnt"(ComponentPose=@n6, Alpha=1, Translation="0,0,0", Rotation=@n5, Scale="1,1,1") {rotationMode=BMM_Additive, rotationSpace=BCS_BoneSpace, alphaBoolBlend {blendOption=Linear}, boneToModify {boneName=L_ear_jnt}}
   n8 = "Transform (Modify) Bone - Bone: R_ear_jnt"(ComponentPose=@n7, Alpha=1, Translation="0,0,0", Rotation=@n5, Scale="1,1,1") {rotationMode=BMM_Additive, rotationSpace=BCS_BoneSpace, alphaBoolBlend {blendOption=Linear}, boneToModify {boneName=R_ear_jnt}}
   Output Pose(Result=Slot 'DefaultSlot'(Source=Component To Local(ComponentPose=@n8))) {layerGroup=None}
   "Sequence Player (additive) 'LlamaRoaster_FiringOverlay_Rifle'"!(StartPosition:float=0) -> (Pose) {sequence=&/LlamaRoaster_FiringOverlay_Rifle}
  ]}
 EventGraph {nodes [
   Event BlueprintUpdateAnimation()
   > n1 = Cast To B_Assault_LlamaRoaster_Athena!(execute, Object=GetOwningActor()) -> (then, CastFailed, AsB Assault Llama Roaster Athena:B Assault Llama Roaster Athena Object Reference) {is=title K2Node_DynamicCast}
   > n2 = Sequence!(execute) -> (then_0, then_1, then_2:Exec)
   > n3 = Branch(Condition="float > float"(A=GetTimeToNextFire!(self:Fort Weapon Object Reference=@n1) -> (ReturnValue:float), B=0) {is=title K2Node_CallFunction})
   > Set IsFiring(IsFiring:bool=true) -> (Output_Get:bool)
   n4 = BreakRotator(InRot=Get BarrelROT())
   > Set BarrelROT(BarrelROT:Rotator=MakeRotator(Roll=float + float(A=@n4, B=10) {is=title K2Node_CommutativeAssociativeBinaryOperator})) -> (Output_Get:Rotator)
   n5 = BreakRotator(InRot=Get BarrelROT())
   > Branch(Condition="float > float"(A=@n5, B=360))
   > Set BarrelROT(BarrelROT=MakeRotator(Roll=float - float(A=@n5, B=360) {is=title K2Node_CallFunction}))
   Set IsFiring(execute=@n3.else, IsFiring=false)
   n6 = BreakRotator(InRot=Get BarrelROT())
   > Branch(Condition="float > float"(A=@n6, B=0))
   > Set BarrelROT(BarrelROT=MakeRotator(Roll=float - float(A=@n6, B=10)))
   Set Max_Overheat(execute=@n2.then_1, Max_Overheat:double=GetOverheatingMaxValue!(self:Fort Weapon Ranged Object Reference=@n1) -> (ReturnValue:float)) -> (Output_Get:double)
   > n7 = Set Overheat_Value(Overheat_Value:double=GetCurrentOverheatValue!(self:Fort Weapon Ranged Object Reference=@n1) -> (ReturnValue:float)) -> (Output_Get:double)
   > Set Overheat_FlapRot(Overheat_FlapRot:Rotator=MakeRotator(Yaw=float * float(A=@n7, B=-1))) -> (Output_Get:Rotator)
   > Branch(Condition=Get IsOverheated() -> (IsOverheated:bool))
   > Set Overheat_FlapRot(Overheat_FlapRot="0, 0, 0")
   GetDisplayName()
   n8 = IsOverheated!(self:Fort Weapon Ranged Object Reference=@n1) -> (ReturnValue:bool)
   PrintString!(execute, InString="To String (Float)"(InDouble="To Float (Boolean)"(InBool=@n8)), bPrintToScreen, bPrintToLog, TextColor, Duration, Key) -> (then)
   n9 = Get Overheat_FlapRot()
   MakeRotator(Yaw=float - float(A=BreakRotator(InRot=@n9).Yaw, B=60))
   n10 = BreakRotator(InRot=Get Overheat_FlapRot())
   Branch(Condition="float >= float"(A=@n10.Yaw, B=60) {is=title K2Node_CallFunction})
   Set IsOverheated(IsOverheated:bool=@n8) -> (Output_Get:bool)
   Set Overheat_FlapRot(Overheat_FlapRot=MakeRotator(Yaw=float - float(A=@n10.Yaw, B=60)))
  ]}
]
variables [
 BarrelROT {type=Rotator}
 BarrelRotation {type=number}
 IsFiring {type=boolean}
 IsOverheated {type=boolean}
 Max_Overheat {type=number}
 Overheat_FlapRot {type=Rotator}
 Overheat_Value {type=number}
]
