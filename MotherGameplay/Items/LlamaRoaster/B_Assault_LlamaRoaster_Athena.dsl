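; Llama Roaster weapon Blueprint: a charge-up beam assault weapon that overheats, driving its montages, FX and audio.
; Overheat events toggle overheat FX and audio, while a server-bound value change feeds the NeutronStar "StarDeath" parameter.
; Impacts reuse one Niagara system with a Cascade fallback, and move a line-shaped audio source, enabled for non-local shooters.
; Recoil helpers sample a vector curve and apply deltas from accumulated recoil to control rotation.
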(event Custom|Apply%20Cam%20Deltas
  (params (RecoilRaw_X :type Float%20%28double-precision%29) (RecoilRaw_Y :type Float%20%28double-precision%29) (RecoilRaw_Z :type Float%20%28double-precision%29))
  ;[1 1 1] Comment
  (bind _asfort_player_pawn (Utilities|Casting|CastToFortPlayerPawn (Game|GetInstigator)))
  (bind _asfort_player_controller (Utilities|Casting|CastToFortPlayerController (Pawn|GetController _asfort_player_pawn)))
  (Fortnite|KeepCGPathOptimizations _asfort_player_controller 10)
  (bind (_returnvalue_roll _returnvalue_pitch _returnvalue_yaw) (Math|Rotator|BreakRotator (Pawn|GetControlRotation _asfort_player_controller :!class /Script/Engine.Controller)))
  (Pawn|SetControlRotation _asfort_player_controller (Math|Rotator|MakeRotator _returnvalue_roll (+ _returnvalue_pitch (* RecoilRaw_Z (Variables|Default|GetVertRecoilMult))) (+ _returnvalue_yaw (* RecoilRaw_X (Variables|Default|GetHorizRecoilMult))))))

(event Custom|UpdateRecoil
  (params (RecoilAlpha :type Float%20%28double-precision%29))
  (bind _returnvalue (* RecoilAlpha 2.0))
  (bind _recoilvectorcurve (Variables|Default|GetRecoilVectorCurve))
  (bind _returnvalue_2 (Math|Curves|GetVectorValue _recoilvectorcurve _returnvalue))
  ;[1 1 1] Comment
  (CallFunction|ApplyCamDeltas (- _returnvalue_2 (Variables|Default|GetAccumulatedRecoil)))
  (Variables|Default|SetAccumulatedRecoil _returnvalue_2))

(event Custom|TriggerDown
  (params (Weapon :type Fort%20Weapon%20Object%20Reference))
  (Development|PrintString "Trigger pressed" (Variables|Default|GetdebugRecoil) (Variables|Default|GetdebugRecoil) "(R=1.000000,G=0.000000,B=0.095176,A=1.000000)" 6.0)
  (Variables|Default|SetTriggerDownTime (Utilities|Time|GetTimeSeconds))
  (Variables|Default|SetbTriggerDown true)
  (Variables|Default|SetAccumulatedRecoil "0, 0, 0")
  (Variables|Default|SetLerpedAlpha 0.0))

(event Custom|CustomEvent_0
  (params (Weapon :type Fort%20Weapon%20Object%20Reference))
  (Development|PrintString (Utilities|String|Append "Trigger released (held for " (Utilities|String|ToString(Float) (- (Utilities|Time|GetTimeSeconds) (Variables|Default|GetTriggerDownTime))) " seconds)") (Variables|Default|GetdebugRecoil) (Variables|Default|GetdebugRecoil) "(R=1.000000,G=0.000000,B=0.091393,A=1.000000)" 6.0)
  ;[1 1 1] Comment
  (Variables|Default|SetbTriggerDown false))

(event FortWeapon|EventOnEquippedAndAttached
  ;[1 1 1] On Equip -> deactivate FX or else they get stuck on if previously heated up
  (|CallParentFunction)
  (Components|Activation|Deactivate (Variables|Default|GetNS_LlamaRoaster_Overheat))
  (Effects|InitializeBeamFX (Variables|Default|GetNiagaraBeam))
  (EventDispatchers|BindEventtoOnPressTrigger (delegate Custom|TriggerDown))
  (EventDispatchers|BindEventtoOnReleaseTrigger (delegate Custom|CustomEvent_0))
  (if (Networking|IsServer)
    (Overheating|AssignOnOverheatValueChanged (delegate Custom|OnOverheatValueChanged_Event_0))))

(event Custom|OnOverheatValueChanged_Event_0
  (params (CurrentValue :type Float%20%28single-precision%29) (MaxValue :type Float%20%28single-precision%29) (State :type EFortWeaponOverheatState%20Enum) (bIsInCooldown :type Boolean))
  ;[1 1 1] On Equip -> deactivate FX or else they get stuck on if previously heated up
  (CallFunction|SetOverheatParam (/ CurrentValue MaxValue)))

(event Custom|SetOverheatParam
  (params (OverheatParamValue :type Float%20%28double-precision%29))
  ;[1 1 1] On Equip -> deactivate FX or else they get stuck on if previously heated up
  (Effects|Components|ParticleSystem|SetFloatParameter (Variables|Default|GetNS_LlamaRoaster_NeutronStar) "StarDeath" OverheatParamValue))

(event Overheating|EventOnStartOverheated
  (bind _ns_llamaroaster_overheat (Variables|Default|GetNS_LlamaRoaster_Overheat))
  ;[1 1 1] Overheat FX
  (Variables|Default|SetbPlayFinishFiringFX false)
  (|CallParentFunction)
  (Components|Activation|Activate _ns_llamaroaster_overheat true)
  (Effects|Components|ParticleSystem|SetFloatParameter _ns_llamaroaster_overheat "AlphaMult" 1.0)
  (Audio|Components|Audio|Play (Variables|Default|GetOverheatAudioLayer) 0.0))

(event Overheating|EventOnStartHeating
  ;[1 1 1] Start Heating
  (|CallParentFunction))

(event FortWeapon|EventOnPlayImpactFX
  (params (HitResult :type Hit%20Result%20Structure%20%28by%20ref%29) (ImpactPhysicalSurface :type EPhysicalSurface%20Enum :default "SurfaceType_Default") (SpawnedPSC :type FXSystem%20Component%20Object%20Reference))
  (bind (_bblockinghit _binitialoverlap _time _distance _location _impactpoint _normal) (Collision|BreakHitResult HitResult))
  (bind _returnvalue (Math|Conversions|RotationFromXVector _normal))
  (bind _impactniagarahandle (Variables|Default|GetImpactNiagaraHandle))
  (bind _fortaudioshapeline (Variables|Default|GetFortAudioShapeLine))
  ;[1 1 1] Play sound a impact point
  (Control|SetStartPoint _fortaudioshapeline (Transformation|GetActorLocation))
  (Control|SetEndPoint _fortaudioshapeline _impactpoint)
  ;[1 1 1] To optimize impact spawning, spawn Llama Roaster specific Impact FX the first time.  Move and restart them if they're already spawned.
  (if (Effects|IsNiagaraEffectEnabled "WeaponImpactEffect" "/MotherGameplay/Items/LlamaRoaster/FX/NS_Assault_LlamaRoaster_Impact.NS_Assault_LlamaRoaster_Impact")
    (Utilities|IsValid _impactniagarahandle
      (:Is%20Valid
        (Transformation|SetWorldLocationAndRotation _impactniagarahandle _impactpoint _returnvalue false false)
        (Niagara|ResetSystem _impactniagarahandle))
      (:Is%20Not%20Valid
        (bind _returnvalue_2 (Niagara|SpawnSystematLocation (Variables|Default|GetLlamaRoasterImpactFXAsset_NiagaraSystem) _impactpoint _returnvalue))
        (Variables|Default|SetImpactNiagaraHandle _returnvalue_2)))
    (else
      (Effects|SpawnEmitteratLocation (Variables|Default|GetLlamaRoasterImpactFXAsset_CascadeEmitter) _impactpoint _returnvalue "1.000000,1.000000,1.000000" true "AutoRelease"))))

(event Overheating|EventOnStartCooling
  ;[1 1 1] Start Cooling
  (|CallParentFunction)
  (Utilities|Time|ClearandInvalidateTimerbyHandle (Variables|Default|GetOverheatFXTimer))
  (Utilities|Time|ClearandInvalidateTimerbyHandle (Variables|Default|GetOverheatFXTimerReplicated)))

(event Overheating|EventOnFinishedCooling
  (bind _ns_llamaroaster_overheat (Variables|Default|GetNS_LlamaRoaster_Overheat))
  ;[1 1 1] Finished Cooling
  (|CallParentFunction)
  (bind (_direction _newtrack_0) (Timeline|AlphaFade PlayFromStart)
    (:Update
      (Effects|Components|ParticleSystem|SetFloatParameter _ns_llamaroaster_overheat "AlphaMult" _newtrack_0)
      (Utilities|FlowControl|Delay 1.0)
      (Components|Activation|Deactivate _ns_llamaroaster_overheat))))

(event EventOnPlayWeaponFireFX
  (params (bPersistentFire :type Boolean) (bSecondaryFire :type Boolean))
  ;[1 1 1] On Play Weapon Fire FX
  (|CallParentFunction bPersistentFire bSecondaryFire)
  (Utilities|FlowControl|Sequence
    (:then_0
      (Components|Activation|Activate (Variables|Default|GetNS_LlamaRoaster_Chargeup_02) false)
      (bind _asfort_player_pawn (Utilities|Casting|CastToFortPlayerPawn (Game|GetInstigator)))
      (if (not (Pawn|IsLocallyControlled _asfort_player_pawn))
        (Shape|Enable (Variables|Default|GetFortAudioShapeLine))))
    (:then_1
      (CallFunction|PlayStartFiringMontage))
    (:then_2
      (Variables|Default|SetbPlayFinishFiringFX true))))

(event EventOnStopWeaponFireFX
  ;[1 1 1] On Stop Weapon Fire FX
  (|CallParentFunction)
  (Components|Activation|Deactivate (Variables|Default|GetNS_LlamaRoaster_Chargeup_02))
  (Shape|Disable (Variables|Default|GetFortAudioShapeLine))
  (CallFunction|PlayFinishedFiringMontage))

(event EventOnStartCharge
  ;[1 1 1] On Start Chargeup
  (CallFunction|PlayChargeupMontage))

(event Custom|PlayStartFiringMontage
  (bind _montage_startfiring_weapon (Variables|Default|GetMontage_StartFiring_Weapon))
  (bind _montage_startfriing_player (Variables|Default|GetMontage_StartFriing_Player))
  ;[1 1 1] Play Start Firing Montage / Note this is when the beam comes out, NOT when the chargeup starts
  (Utilities|IsValid _montage_startfiring_weapon
    (:Is%20Valid
      (Animation|Montage|PlayMontage (Variables|Visual|GetWeaponMesh) _montage_startfiring_weapon))
    (:Is%20Not%20Valid
      _))
  (Utilities|IsValid _montage_startfriing_player
    (:Is%20Valid
      (bind _asfort_player_pawn (Utilities|Casting|CastToFortPlayerPawn (Game|GetInstigator)))
      (Animation|Montage|PlayMontage (Variables|/Script/Engine.Character|GetMesh _asfort_player_pawn) _montage_startfriing_player))))

(event Custom|PlayFinishedFiringMontage
  (bind _montage_finishfiring_weapon (Variables|Default|GetMontage_FinishFiring_Weapon))
  (bind _montage_finishfiring_player (Variables|Default|GetMontage_FinishFiring_Player))
  ;[1 1 1] Play Finished Firing Montage
  (if (Variables|Default|GetbPlayFinishFiringFX)
    (Utilities|IsValid _montage_finishfiring_weapon
      (:Is%20Valid
        (Animation|Montage|PlayMontage (Variables|Visual|GetWeaponMesh) _montage_finishfiring_weapon))
      (:Is%20Not%20Valid
        _))
    (Utilities|IsValid _montage_finishfiring_player
      (:Is%20Valid
        (bind _asfort_player_pawn (Utilities|Casting|CastToFortPlayerPawn (Game|GetInstigator)))
        (Animation|Montage|PlayMontage (Variables|/Script/Engine.Character|GetMesh _asfort_player_pawn) _montage_finishfiring_player)))))

(event Custom|PlayChargeupMontage
  (bind _montage_chargeup_weapon (Variables|Default|GetMontage_Chargeup_Weapon))
  (bind _montage_chargeup_player (Variables|Default|GetMontage_Chargeup_Player))
  ;[1 1 1] Play Chargeup Montage
  (Utilities|IsValid _montage_chargeup_weapon
    (:Is%20Valid
      (Animation|Montage|PlayMontage (Variables|Visual|GetWeaponMesh) _montage_chargeup_weapon)
      (Utilities|IsValid _montage_chargeup_player
        (:Is%20Valid
          (bind _asfort_player_pawn (Utilities|Casting|CastToFortPlayerPawn (Game|GetInstigator)))
          (Animation|Montage|PlayMontage (Variables|/Script/Engine.Character|GetMesh _asfort_player_pawn) _montage_chargeup_player))))))

(fn ConstructionScript
  (|CallParentFunction))
