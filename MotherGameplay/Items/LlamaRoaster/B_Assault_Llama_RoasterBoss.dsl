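; Boss variant of the Llama Roaster assault weapon Blueprint, which recolors the weapon's emissive glow.
; The construction script creates a dynamic instance of MI_LlamaRoaster on material slot 0 of the parent's weapon mesh and stores it.
; BeginPlay then sets that instance's EmissiveMultiplier vector to an overbright red-orange.
; Both events call the parent implementation, and Tick adds no logic of its own.
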(event EventBeginPlay
  (|CallParentFunction)
  (Rendering|Material|SetVectorParameterValue (Variables|Default|GetDynamicMaterialInstanceBoss) "EmissiveMultiplier" "(R=3.000000,G=0.382995,B=0.133973,A=1.000000)" :!class /Script/Engine.MaterialInstanceDynamic))

(event EventTick
  (params (DeltaSeconds :type Float%20%28single-precision%29))
  (|CallParentFunction DeltaSeconds) ; This node is disabled and will not be called.\nDrag off pins to build functionality.
)

(fn ConstructionScript
  (|CallParentFunction)
  (bind _returnvalue (Rendering|Material|CreateDynamicMaterialInstance (Variables|Visual|GetWeaponMesh) 0 "/MotherGameplay/Items/LlamaRoaster/Assets/Materials/MI_LlamaRoaster.MI_LlamaRoaster" "None" :!class /Script/Engine.PrimitiveComponent))
  (Variables|Default|SetDynamicMaterialInstanceBoss _returnvalue))
