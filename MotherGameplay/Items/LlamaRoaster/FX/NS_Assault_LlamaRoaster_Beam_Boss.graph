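# Niagara beam effect for the boss variant of the LlamaRoaster assault item, running from a fixed local offset to User.BeamEnd_0.
# MainRibbon (two renderers) and Plasma (panning material) place partitioned ribbon particles along a Hermite spline, while RingsMeshes sends fading, pulsing meshes down it.
# All three add waveform wobble, driven by distance from User.BeamStart_0_0, rotated by System.BeamOrientation and tapered at both ends.
# The end-ball sprite emitter is disabled.
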NiagaraSystem /MotherGameplay/Items/LlamaRoaster/FX/NS_Assault_LlamaRoaster_Beam_Boss
dependencies {
 usedDataInterfaces [&NiagaraDataInterfaceCurve]
 usedDynamicInputs [&FindQuatBetweenTwoVectors]
 usedModules [
  &SystemState
  &self:SystemSpawnScript.SystemScriptSource.SystemScriptGraph.NiagaraNodeAssignment_0.SetVariables_#1
  &self:SystemSpawnScript.SystemScriptSource.SystemScriptGraph.NiagaraNodeAssignment_1.SetVariables_#2
 ]
}
emitters {
 Ball_At_the_end {
  bEnabled=false
  emitterUpdateScript [
   &EmitterState()
   &SpawnBurst_Instantaneous(Spawn Count=1, Use Spawn Probability=false)
  ]
  particleSpawnScript [&InitializeParticle(
    Color Mode=::NewEnumerator0
    Lifetime=0.7
    Mesh Scale Mode=::NewEnumerator3
    Mesh Uniform Scale=1000
    Position=$User.BeamEnd_0
    Position Mode=::NewEnumerator1
    UsePositionOffset=false
    Write Material Random=true
   )]
  particleUpdateScript [&ParticleState()]
  rendererClasses [&NiagaraSpriteRendererProperties]
  renderers [{rendererClass=&NiagaraSpriteRendererProperties}]
 }
 MainRibbon {
  emitterSpawnScript [/Game/Effects/Niagara/Modules/Partition/FN_PartitionAttributes(Number of Bone Sets=1, Particles Per Partition=30)]
  emitterUpdateScript [
   &EmitterState()
   &SpawnBurst_Instantaneous(Spawn Count=$Emitter.FN_PartitionAttributes.Spawn Total, Use Spawn Probability=false)
  ]
  particleSpawnScript [
   /Niagara/Modules/Spawn/Initialization/InitializeParticle(
    Lifetime=5
    Module.Write Color=true
    Module.Write Lifetime=true
    Module.Write Mass=true
    Module.Write Material Random=true
    Module.Write Position=true
    Module.Write Scale=false
    Module.Write SpriteRotation=false
    Module.Write SpriteSize=true
    Sprite Size=<30 30>
   )
   &InitializeRibbon(
    Lifetime=0.7
    Module.Write Color=true
    Module.Write Lifetime=true
    Module.Write Mass=true
    Module.Write Material Random=true
    Module.Write Position=false
    Module.Write RibbonTwist=false
    Module.Write RibbonWidth=true
    Ribbon Width=11
   )
   /Game/Effects/Niagara/Modules/Partition/FN_Particle_partition_v5/FN_Particle_Partition_v5(Enable Ribbon ID=true)
   /SetVariables(Particles.Beam2_Width=45) {isSetParameters=true}
   /SetVariables(Particles.Renderer2_Color:LinearColor=<rgba 3 0.2002 0 0.4>) {isSetParameters=true}
   /SetVariables(Particles.WobbleOverLength:NiagaraFloat=&FloatFromCurve(
     FloatCurve=DataInterface() {propertyValues {Curve {keys [
         {arriveTangent=5.2404, interpMode=RCIM_Cubic, leaveTangent=5.2404, tangentMode=RCTM_User, time=0, value=0}
         <at 0.5 1 RCIM_Cubic>
         {arriveTangent=-4.4363, interpMode=RCIM_Cubic, leaveTangent=-4.4363, tangentMode=RCTM_User, time=1, value=0}
        ]}}}
     CurveIndex=$Particles.Partition.NormalizedID
    )) {isSetParameters=true, module=SetVariables_#3}
  ]
  particleUpdateScript [
   &ParticleState()
   &SystemLocation(MaskSpawnGroup=false)
   /Game/Effects/Niagara/Modules/Curves/FN_HermiteCubicSpline_v1/FN_HermiteCubicSpline(
    Position Blend:NiagaraFloat=$Particles.Initial.Partition.NormalizedID
    Position End:Vector3f=/Niagara/DynamicInputs/Transforms/TransformPosition(Position=Linked() {linkedVariable {name=User.BeamEnd_0, type=NiagaraPosition}}, Source Space=::World, Destination Space=::Local)
    Position Start:Vector3f=<0 72 8.5>
    Write Particles Position=true
   )
   /SetVariables(Particles.ParticleDistance:NiagaraFloat=&VectorLength(Vector=&Subtract_Vector(A=Linked() {linkedVariable {name=Particles.Position, type=NiagaraPosition}}, B=$User.BeamStart_0_0))) {isSetParameters=true}
   /SetVariables(Particles.Position=&Add_Vector(
     A=$Output.FN_HermiteCubicSpline.Position
     B=&Multiply_VectorByFloat(
      Vector=&RotateVectorByQuaternion(
       Vector To Rotate=&MakeVector(
        X=0
        Y=&Waveform(
         Global Input Position=&Multiply_Float(A=$Particles.ParticleDistance, B=0.005)
         Global Amplitude Scale=10
         "Waveform [1]"=::NewEnumerator0
         "[1] Frequency"=1
         "[1] Amplitude Min/Max"=<0 1>
         "[1] Bias"=-0.5
         "[2] Phase"=$Emitter.Age
         "[3] Phase"=$Emitter.Age
        )
        Z=&Waveform(
         Global Input Position=&Multiply_Float(A=$Particles.ParticleDistance, B=0.005)
         Global Amplitude Scale=10
         "Waveform [1]"=::NewEnumerator1
         "[1] Frequency"=1
         "[1] Amplitude Min/Max"=<0 1>
         "[1] Bias"=-0.5
         "[2] Phase"=$Emitter.Age
         "[3] Phase"=$Emitter.Age
        )
       )
       Quaternion=$System.BeamOrientation
      )
      Float=$Particles.WobbleOverLength
     )
    )) {isSetParameters=true, module=SetVariables_#4}
   /SetVariables(Particles.Renderer2_Position:Vector3f=&Add_Vector(
     A=$Output.FN_HermiteCubicSpline.Position
     B=&Multiply_VectorByFloat(
      Vector=&RotateVectorByQuaternion(
       Vector To Rotate=&MakeVector(
        X=0
        Y=&Sine(
         Normalized Angle=&Add_Float(A=$Particles.Partition.NormalizedID, B=&Multiply_Float(A=$Emitter.Age, B=3))
         Period=&Divide_Float(A=$System.BeamLength, B=1500)
         Scale=1
         Bias=0
        )
        Z=&Cosine(
         Normalized Angle=&Add_Float(A=$Particles.Partition.NormalizedID, B=&Multiply_Float(A=$Emitter.Age, B=3))
         Period=&Divide_Float(A=$System.BeamLength, B=1500)
         Scale=1
         Bias=0
        )
       )
       Quaternion=$System.BeamOrientation
      )
      Float=$Particles.WobbleOverLength
     )
    )) {enabled=false, isSetParameters=true}
   &ScaleRibbonWidth(Ribbon Width Scale=&Sine(
     Normalized Angle=&Add_Float(A=$Particles.Partition.NormalizedID, B=$Emitter.Age)
     Period=0.25
     Scale=0.25
     Bias=0.5
    ))
  ]
  rendererClasses [&NiagaraRibbonRendererProperties]
  renderers [
   {rendererClass=&NiagaraRibbonRendererProperties}
   {rendererClass=&NiagaraRibbonRendererProperties, rendererIndex=1}
  ]
 }
 Plasma {
  emitterSpawnScript [&FN_PartitionAttributes(Number of Bone Sets=1, Particles Per Partition=12)]
  emitterUpdateScript [
   &EmitterState()
   &SpawnBurst_Instantaneous(Spawn Count=$Emitter.FN_PartitionAttributes.Spawn Total, Use Spawn Probability=false)
  ]
  particleSpawnScript [
   &InitializeParticle(
    Lifetime=5
    Module.Write Color=true
    Module.Write Lifetime=true
    Module.Write Mass=true
    Module.Write Material Random=true
    Module.Write Position=true
    Module.Write Scale=false
    Module.Write SpriteRotation=false
    Module.Write SpriteSize=true
    Sprite Size=<30 30>
   )
   &FN_Particle_Partition_v5(Enable Ribbon ID=true)
   &InitializeRibbon(
    Lifetime=0.7
    Module.Write Color=true
    Module.Write Lifetime=true
    Module.Write Mass=true
    Module.Write Material Random=true
    Module.Write Position=false
    Module.Write RibbonTwist=false
    Module.Write RibbonWidth=true
    Ribbon Width=15
   )
   /SetVariables(Particles.WobbleOverLength:NiagaraFloat=&FloatFromCurve(
     FloatCurve=DataInterface() {propertyValues {Curve {keys [
         {arriveTangent=5.2404, interpMode=RCIM_Cubic, leaveTangent=5.2404, tangentMode=RCTM_User, time=0, value=0}
         <at 0.5 1 RCIM_Cubic>
         {arriveTangent=-4.4363, interpMode=RCIM_Cubic, leaveTangent=-4.4363, tangentMode=RCTM_User, time=1, value=0}
        ]}}}
     CurveIndex=$Particles.Partition.NormalizedID
    )) {isSetParameters=true, module=SetVariables_#3}
  ]
  particleUpdateScript [
   &ParticleState()
   &SystemLocation(MaskSpawnGroup=false)
   /SetVariables(Particles.RibbonUVDistance=0.05) {isSetParameters=true}
   &FN_HermiteCubicSpline(
    Position Blend:NiagaraFloat=$Particles.Initial.Partition.NormalizedID
    Position End:Vector3f=&TransformPosition(Position=Linked() {linkedVariable {name=User.BeamEnd_0, type=NiagaraPosition}}, Source Space=::World, Destination Space=::Local)
    Position Start:Vector3f=<0 72 8.5>
    Write Particles Position=true
   )
   /SetVariables(Particles.ParticleDistance:NiagaraFloat=&VectorLength(Vector=&Subtract_Vector(A=Linked() {linkedVariable {name=Particles.Position, type=NiagaraPosition}}, B=$User.BeamStart_0_0))) {isSetParameters=true}
   /SetVariables(Particles.Position=&Add_Vector(
     A=$Output.FN_HermiteCubicSpline.Position
     B=&Multiply_VectorByFloat(
      Vector=&RotateVectorByQuaternion(
       Vector To Rotate=&MakeVector(
        X=0
        Y=&Sine(
         Normalized Angle=&Add_Float(A=$Particles.Partition.NormalizedID, B=&Multiply_Float(A=$Emitter.Age, B=3))
         Period=&Divide_Float(A=$System.BeamLength, B=1500)
         Scale=3
         Bias=0
        )
        Z=&Cosine(
         Normalized Angle=&Add_Float(A=$Particles.Partition.NormalizedID, B=&Multiply_Float(A=$Emitter.Age, B=3))
         Period=&Divide_Float(A=$System.BeamLength, B=1500)
         Scale=3
         Bias=0
        )
       )
       Quaternion=$System.BeamOrientation
      )
      Float=$Particles.WobbleOverLength
     )
    )) {enabled=false, isSetParameters=true, module=SetVariables_#4}
   /SetVariables(Particles.Position=&Add_Vector(
     A=$Output.FN_HermiteCubicSpline.Position
     B=&Multiply_VectorByFloat(
      Vector=&RotateVectorByQuaternion(
       Vector To Rotate=&MakeVector(
        X=0
        Y=&Waveform(
         Global Input Position=&Multiply_Float(A=$Particles.ParticleDistance, B=0.005)
         Global Amplitude Scale=20
         "Waveform [1]"=::NewEnumerator0
         "[1] Frequency"=1
         "[1] Amplitude Min/Max"=<0 1>
         "[1] Bias"=-0.5
         "[2] Phase"=$Emitter.Age
         "[3] Phase"=$Emitter.Age
        )
        Z=&Waveform(
         Global Input Position=&Multiply_Float(A=$Particles.ParticleDistance, B=0.005)
         Global Amplitude Scale=20
         "Waveform [1]"=::NewEnumerator1
         "[1] Frequency"=1
         "[1] Amplitude Min/Max"=<0 1>
         "[1] Bias"=-0.5
         "[2] Phase"=$Emitter.Age
         "[3] Phase"=$Emitter.Age
        )
       )
       Quaternion=$System.BeamOrientation
      )
      Float=$Particles.WobbleOverLength
     )
    )) {isSetParameters=true}
   &ScaleRibbonWidth(Ribbon Width Scale=&Sine(
     Normalized Angle=&Add_Float(A=$Particles.Partition.NormalizedID, B=$Emitter.Age)
     Period=0.25
     Scale=0.25
     Bias=0.5
    ))
   &FN_RibbonMaterial(
    Color=<rgba 5 0.5703 0.5205 1>
    Macro Distortion Amount=0.1
    Macro Pan Speed=1
    Macro Scale U=10
    Opacity Pan U Speed=3
    Opacity Scale U=250
    UV Basis:NiagaraFloat=$Particles.Partition.NormalizedID
   )
   /SetVariables(Particles.RibbonUVDistance=10) {enabled=false, isSetParameters=true}
  ]
  rendererClasses [&NiagaraRibbonRendererProperties]
  renderers [{rendererClass=&NiagaraRibbonRendererProperties}]
 }
 RingsMeshes {
  emitterSpawnScript [&FN_PartitionAttributes(Number of Bone Sets=1, Particles Per Partition=12)]
  emitterUpdateScript [
   &EmitterState()
   &SpawnBurst_Instantaneous(Spawn Count=$Emitter.FN_PartitionAttributes.Spawn Total, Use Spawn Probability=false) {enabled=false}
   &SpawnRate(SpawnRate=50, Use Spawn Probability=false)
  ]
  particleSpawnScript [
   &InitializeParticle(
    Color=<rgba 1 0.1676 0.1311 0.15>
    Lifetime=0.5
    Mesh Scale=&RandomRangeVector(Minimum=<0.25 0.25 1>, Maximum=<0.5 0.5 3>)
    Module.Write Color=true
    Module.Write Lifetime=true
    Module.Write Mass=true
    Module.Write Material Random=true
    Module.Write Position=true
    Module.Write Scale=true
    Module.Write SpriteRotation=false
    Module.Write SpriteSize=false
   )
   &FN_Particle_Partition_v5(Enable Ribbon ID=true)
   &InitializeRibbon(
    Color=<rgba 1 0 0.6104 0.15>
    Lifetime=0.7
    Module.Write Color=true
    Module.Write Lifetime=true
    Module.Write Mass=true
    Module.Write Material Random=true
    Module.Write Position=false
    Module.Write RibbonTwist=false
    Module.Write RibbonWidth=true
    Ribbon Width=14
   ) {enabled=false}
   &InitialMeshOrientation(
    Orientation Axis=<0 0 1>
    Orientation Coordinate Space=::World
    Orientation Vector=&RotateVectorByQuaternion(Vector To Rotate=<1 0 0>, Quaternion=$System.BeamOrientation)
    Use Orientation Vector=true
    Use Rotation Vector=false
   )
   /SetVariables(Particles.WobbleOverLength:NiagaraFloat=&FloatFromCurve(
     FloatCurve=DataInterface() {propertyValues {Curve {keys [
         {arriveTangent=5.2404, interpMode=RCIM_Cubic, leaveTangent=5.2404, tangentMode=RCTM_User, time=0, value=0}
         <at 0.5 1 RCIM_Cubic>
         {arriveTangent=-4.4363, interpMode=RCIM_Cubic, leaveTangent=-4.4363, tangentMode=RCTM_User, time=1, value=0}
        ]}}}
     CurveIndex=$Particles.Partition.NormalizedID
    )) {isSetParameters=true}
  ]
  particleUpdateScript [
   &ParticleState()
   &FN_HermiteCubicSpline(
    Position Blend:NiagaraFloat=$Particles.NormalizedAge
    Position End:Vector3f=&TransformPosition(Position=Linked() {linkedVariable {name=User.BeamEnd_0, type=NiagaraPosition}}, Source Space=::World, Destination Space=::Local)
    Position Start:Vector3f=<0 72 8.5>
    Write Particles Position=true
   )
   /SetVariables(Particles.ParticleDistance:NiagaraFloat=&VectorLength(Vector=&Subtract_Vector(A=Linked() {linkedVariable {name=Particles.Position, type=NiagaraPosition}}, B=$User.BeamStart_0_0))) {isSetParameters=true}
   /SetVariables(Particles.Position=&Add_Vector(
     A=$Output.FN_HermiteCubicSpline.Position
     B=&Multiply_VectorByFloat(
      Vector=&RotateVectorByQuaternion(
       Vector To Rotate=&MakeVector(
        X=0
        Y=&Sine(
         Normalized Angle=&Add_Float(A=$Particles.Partition.NormalizedID, B=&Multiply_Float(A=$Emitter.Age, B=3))
         Period=&Divide_Float(A=$System.BeamLength, B=1500)
         Scale=1.3
         Bias=0
        )
        Z=&Cosine(
         Normalized Angle=&Add_Float(A=$Particles.Partition.NormalizedID, B=&Multiply_Float(A=$Emitter.Age, B=3))
         Period=&Divide_Float(A=$System.BeamLength, B=1500)
         Scale=1.3
         Bias=0
        )
       )
       Quaternion=$System.BeamOrientation
      )
      Float=$Particles.WobbleOverLength
     )
    )) {enabled=false, isSetParameters=true, module=SetVariables_#4}
   /SetVariables(Particles.Position=&Add_Vector(
     A=$Output.FN_HermiteCubicSpline.Position
     B=&Multiply_VectorByFloat(
      Vector=&RotateVectorByQuaternion(
       Vector To Rotate=&MakeVector(
        X=0
        Y=&Waveform(
         Global Input Position=&Multiply_Float(A=$Particles.ParticleDistance, B=0.005)
         Global Amplitude Scale=5
         "Waveform [1]"=::NewEnumerator0
         "[1] Frequency"=1
         "[1] Amplitude Min/Max"=<0 1>
         "[1] Bias"=-0.5
         "[2] Phase"=$Emitter.Age
         "[3] Phase"=$Emitter.Age
        )
        Z=&Waveform(
         Global Input Position=&Multiply_Float(A=$Particles.ParticleDistance, B=0.005)
         Global Amplitude Scale=5
         "Waveform [1]"=::NewEnumerator1
         "[1] Frequency"=1
         "[1] Amplitude Min/Max"=<0 1>
         "[1] Bias"=-0.5
         "[2] Phase"=$Emitter.Age
         "[3] Phase"=$Emitter.Age
        )
       )
       Quaternion=$System.BeamOrientation
      )
      Float=$Particles.WobbleOverLength
     )
    )) {isSetParameters=true}
   &ScaleMeshSize(Scale Factor=&VectorFromFloat(Value=&Add_Float(
      A=&Sine(Normalized Angle=&Subtract_Float(A=$Particles.NormalizedAge, B=$Emitter.Age), Period=0.5, Scale=0.25, Bias=0)
      B=1
     ))) {enabled=false}
   &ScaleMeshSize(Scale Factor=&MakeVector(
     X=&Add_Float(
      A=&Sine(Normalized Angle=&Subtract_Float(A=$Particles.NormalizedAge, B=$Emitter.Age), Period=0.5, Scale=0.25, Bias=0)
      B=1
     )
     Y=&Add_Float(
      A=&Sine(Normalized Angle=&Subtract_Float(A=$Particles.NormalizedAge, B=$Emitter.Age), Period=0.5, Scale=0.25, Bias=0)
      B=1
     )
     Z=0.75
    )) {module=ScaleMeshSize001}
   &ScaleColor(
    Scale Alpha=&FloatFromCurve(FloatCurve=DataInterface() {propertyValues {Curve {keys [
         {arriveTangent=1.1318, interpMode=RCIM_Cubic, leaveTangent=1.1318, tangentMode=RCTM_User, time=0, value=0}
         {arriveTangent=-0.9757, interpMode=RCIM_Cubic, leaveTangent=-0.9757, tangentMode=RCTM_User, time=0.1, value=1}
         {arriveTangent=-0.7753, interpMode=RCIM_Cubic, leaveTangent=-0.7753, tangentMode=RCTM_User, time=1, value=0}
        ]}}})
    Scale RGB=&VectorFromFloat(Value=&FloatFromCurve(
      FloatCurve=DataInterface() {propertyValues {Curve {keys [{tangentMode=RCTM_User, time=0, value=1}, {tangentMode=RCTM_User, time=1, value=0}]}}}
      Scale Curve=3
     ))
    ScaleA=true
    ScaleRGB=true
   )
  ]
  rendererClasses [&NiagaraMeshRendererProperties]
  renderers [{rendererClass=&NiagaraMeshRendererProperties, rendererIndex=0}]
 }
}
summary {
 systemName=NS_Assault_LlamaRoaster_Beam_Boss
 emitters [
  {emitterName=MainRibbon, rendererClasses [&NiagaraRibbonRendererProperties]}
  {emitterName=RingsMeshes, rendererClasses [&NiagaraMeshRendererProperties]}
  {emitterName=Plasma, rendererClasses [&NiagaraRibbonRendererProperties]}
  {bEnabled=false, emitterName=Ball_At_the_end, rendererClasses [&NiagaraSpriteRendererProperties]}
 ]
 userVariables [
  User.BeamEnd_0 {type=NiagaraPosition, description="", defaultValue {}}
  User.BeamStart_0_0 {type=Vector3f, description="", defaultValue {struct=Vector3f, value=<1000 0 1000>}}
 ]
}
