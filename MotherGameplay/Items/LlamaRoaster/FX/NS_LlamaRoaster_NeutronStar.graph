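# Niagara effect for the LlamaRoaster item: a miniature star system that collapses into a neutron star.
# A sun with mesh, sprite and flickering light layers anchors three planets on tilted sine/cosine orbits, indexed per particle through arrays.
# User.StarDeath drives the collapse, shrinking bodies and orbits to the centre, shifting colours towards bright blue and growing an extra sun sprite.
# The same parameter fades in and enlarges continuously spawned smoke sprites.
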NiagaraSystem /MotherGameplay/Items/LlamaRoaster/FX/NS_LlamaRoaster_NeutronStar
dependencies {
 usedDataInterfaces [&NiagaraDataInterfaceCurve, &NiagaraDataInterfaceArrayFloat, &NiagaraDataInterfaceArrayFloat3]
 usedDynamicInputs []
 usedModules [&SystemState]
}
emitters {
 Planets {
  emitterUpdateScript [
   &EmitterState(
    Life Cycle Mode=::NewEnumerator1
    Loop Behavior=::NewEnumerator1
    Loop Duration Mode=::NewEnumerator1
    Reset Age On Awaken=true
    Scalability Mode=::NewEnumerator1
    UseLoopDelay=false
   )
   &SpawnBurst_Instantaneous(Spawn Count=3, Use Spawn Probability=false)
  ]
  particleSpawnScript [
   &InitializeParticle(
    Color Mode=::NewEnumerator0
    Mesh Scale Mode=::NewEnumerator3
    Mesh Uniform Scale=&SelectFloatFromArray(
     Direct Array Index=&ReturnParticleIDAttributes()
     Float Selection Array=DataInterface() {propertyValues {FloatData [0.18,0.12,0.1,0.075,0.05]}}
    )
    Sprite Size Mode=::NewEnumerator3
    Uniform Sprite Size=&Multiply_Float(
     A=15
     B=&SelectFloatFromArray(
      Direct Array Index=&ReturnParticleIDAttributes()
      Float Selection Array=DataInterface() {propertyValues {FloatData [0.18,0.12,0.1,0.075,0.05]}}
     )
    )
    UsePositionOffset=false
    Write Material Random=true
   )
   /SetVariables(Particles.Position=&Add_Vector(
     A=Linked() {linkedVariable {name=Engine.Owner.Position, type=NiagaraPosition}}
     B=&SelectVectorFromArray(
      Direct Array Index=&ReturnParticleIDAttributes()
      Vector Selection Array=DataInterface() {propertyValues {FloatData [<7 0 1>, <-5 -3 -1.5>, <0 7 0>, <3 2 2>, <-2 -2 -3>]}}
     )
    )) {enabled=false, isSetParameters=true}
  ]
  particleUpdateScript [
   &ParticleState(Kill Particles When Lifetime Has Elapsed=false, Let Infinitely Lived Particles Die When Emitter Deactivates=true)
   /SetVariables(Particles.Renderer2_Color:LinearColor=&LerpLinearColors(StartColor=<rgba 0.1063 0 1 1>, EndColor=<rgba 0 0.3672 1 1>, LerpFactor=$User.StarDeath)) {isSetParameters=true}
   &ScaleColor(
    Scale RGB=&MakeVectorFromLinearColorRGB(Color=&LerpLinearColors(StartColor=<rgba 1 1 1 1>, EndColor=<rgba 0 18.3599 50 1>, LerpFactor=$User.StarDeath))
    ScaleA=true
    ScaleRGB=true
   )
   /SetVariables(Particles.Position=&MakeVector(
     X=&Sine(
      Normalized Angle=$Emitter.Age
      Period=&SelectFloatFromArray(
       Direct Array Index=&ReturnParticleIDAttributes()
       Float Selection Array=DataInterface() {propertyValues {FloatData [1,1.5,2,2.5,3]}}
      )
      Scale=&SelectFloatFromArray(
       Direct Array Index=&ReturnParticleIDAttributes()
       Float Selection Array=DataInterface() {propertyValues {FloatData [5,7,9,11,13]}}
      )
      Bias=0
     )
     Y=&Cosine(
      Normalized Angle=$Emitter.Age
      Period=&SelectFloatFromArray(
       Direct Array Index=&ReturnParticleIDAttributes()
       Float Selection Array=DataInterface() {propertyValues {FloatData [1,1.5,2,2.5,3]}}
      )
      Scale=&SelectFloatFromArray(
       Direct Array Index=&ReturnParticleIDAttributes()
       Float Selection Array=DataInterface() {propertyValues {FloatData [5,7,9,11,13]}}
      )
      Bias=0
     )
     Z=0
    )) {enabled=false, isSetParameters=true}
   &UpdateMeshOrientation(Rotation Vector=<0 0 1>)
   /SetVariables(Particles.Position=&RotateVectorByQuaternion(
     Vector To Rotate=&MakeVector(
      X=&Sine(
       Normalized Angle=$Emitter.Age
       Period=&SelectFloatFromArray(
        Direct Array Index=&ReturnParticleIDAttributes()
        Float Selection Array=DataInterface() {propertyValues {FloatData [3.5,2.3,2.5]}}
       )
       Scale=&Lerp_Float(
        A=&SelectFloatFromArray(
         Direct Array Index=&ReturnParticleIDAttributes()
         Float Selection Array=DataInterface() {propertyValues {FloatData [3.25,3.3,3.35]}}
        )
        B=0
        Alpha=$User.StarDeath
       )
       Bias=0
      )
      Y=&Cosine(
       Normalized Angle=$Emitter.Age
       Period=&SelectFloatFromArray(
        Direct Array Index=&ReturnParticleIDAttributes()
        Float Selection Array=DataInterface() {propertyValues {FloatData [3.5,2.3,2.5]}}
       )
       Scale=&Lerp_Float(
        A=&SelectFloatFromArray(
         Direct Array Index=&ReturnParticleIDAttributes()
         Float Selection Array=DataInterface() {propertyValues {FloatData [3.25,3.3,3.35]}}
        )
        B=0
        Alpha=$User.StarDeath
       )
       Bias=0
      )
      Z=0
     )
     Quaternion=&MakeQuatFromAxisAngle(
      Axis=<1 0 0>
      AngleInDegrees=&SelectFloatFromArray(
       Direct Array Index=&ReturnParticleIDAttributes()
       Float Selection Array=DataInterface() {propertyValues {FloatData [10,-35,5]}}
      )
     )
    )) {isSetParameters=true}
   &ScaleMeshSize(Scale Factor=&VectorFromFloat(Value=&Lerp_Float(A=1, B=0.2, Alpha=$User.StarDeath)))
  ]
  rendererClasses [&NiagaraMeshRendererProperties, &NiagaraSpriteRendererProperties]
  renderers [
   {rendererClass=&NiagaraMeshRendererProperties, rendererIndex=0}
   {rendererClass=&NiagaraSpriteRendererProperties, rendererIndex=1}
  ]
 }
 SMoke {
  emitterUpdateScript [
   &EmitterState()
   &SpawnRate(SpawnRate=3, Use Spawn Probability=false)
  ]
  particleSpawnScript [
   /Niagara/Modules/Spawn/Initialization/InitializeParticle(
    Color=&LerpLinearColors(StartColor=<rgba 0 5.3229 10 0.5>, EndColor=<rgba 0 0.6787 19 0.25>, LerpFactor=0.75)
    Lifetime=0.7
    Module.Write Color=true
    Module.Write Lifetime=true
    Module.Write Mass=true
    Module.Write Material Random=true
    Module.Write Position=false
    Module.Write Scale=false
    Module.Write SpriteRotation=true
    Module.Write SpriteSize=true
    Position=&SimulationPosition()
    Sprite Rotation=&RandomRangeFloat(Minimum=-180, Maximum=180)
    Sprite Size=&Vector2DFromFloat(Value=15)
   )
   /SetVariables(Particles.RandomFrame=&RandomRangeInt(Minimum=0, Maximum=63)) {enabled=false, isSetParameters=true}
  ]
  particleUpdateScript [
   &ParticleState()
   &CameraOffset(Camera Offset Amount=3)
   &FN_SubUVAnimation(End Frame=&Add_Integer(A=$Particles.RandomFrame, B=-35), Start Frame=$Particles.RandomFrame) {enabled=false}
   &ScaleSpriteSize(Scale Factor=&Vector2DFromFloat(Value=&FloatFromCurve(
      FloatCurve=DataInterface() {propertyValues {Curve {keys [{tangentMode=RCTM_User, time=0, value=1}, {tangentMode=RCTM_User, time=1, value=0}]}}}
      Scale Curve=&FloatFromCurve(
       FloatCurve=DataInterface() {propertyValues {Curve {keys [
           <at 0 0 RCIM_Cubic>
           {arriveTangent=1.875, interpMode=RCIM_Cubic, leaveTangent=1.875, time=0.2, value=0.1}
           {arriveTangent=1.125, interpMode=RCIM_Cubic, leaveTangent=1.125, time=0.4, value=0.75}
           <at 1 1 RCIM_Cubic>
          ]}}}
       CurveIndex=$User.StarDeath
      )
     )))
   &ScaleColor(
    Scale Alpha=&FloatFromCurve(
     FloatCurve=DataInterface() {propertyValues {Curve {keys [
         {arriveTangent=4.5667, interpMode=RCIM_Cubic, leaveTangent=4.5667, tangentMode=RCTM_User, time=0.0023, value=0}
         {arriveTangent=-0.0044, interpMode=RCIM_Cubic, leaveTangent=-0.0044, tangentMode=RCTM_User, time=1, value=1.5}
        ]}}}
     Scale Curve=&Lerp_Float(A=0, B=1, Alpha=$User.StarDeath)
    )
    ScaleA=true
    ScaleRGB=false
   )
   &DynamicMaterialParameters(
    Index 0 Param 1=$Particles.MaterialRandom
    Index 0 Param 2=&RandomRangeFloat(Minimum=0, Maximum=1)
    Module.Param0WriteEnabled=true
    Module.Param1WriteEnabled=true
    Module.Param2WriteEnabled=true
    Module.Param3WriteEnabled=false
   )
  ]
  rendererClasses [&NiagaraSpriteRendererProperties]
  renderers [{rendererClass=&NiagaraSpriteRendererProperties}]
 }
 Sun {
  emitterUpdateScript [
   &EmitterState(
    Life Cycle Mode=::NewEnumerator1
    Loop Behavior=::NewEnumerator1
    Loop Duration Mode=::NewEnumerator1
    Reset Age On Awaken=true
    Scalability Mode=::NewEnumerator1
    UseLoopDelay=false
   )
   &SpawnBurst_Instantaneous(Spawn Count=1, Use Spawn Probability=false)
  ]
  particleSpawnScript [
   /Niagara/Modules/Spawn/Initialization/V2/InitializeParticle(
    Mesh Scale Mode=::NewEnumerator3
    Mesh Uniform Scale=0.05
    Sprite Size Mode=::NewEnumerator3
    Uniform Sprite Size=7.5
    UsePositionOffset=false
    Write Material Random=true
   )
   /SetVariables(Particles.Renderer2_CamOffset=-25) {enabled=false, isSetParameters=true}
  ]
  particleUpdateScript [
   &ParticleState(
    Kill Particles When Lifetime Has Elapsed=false
    Let Infinitely Lived Particles Die When Emitter Deactivates=true
    Loop Particles Lifetime=true
   )
   &ScaleColor(
    Scale RGB=&MakeVectorFromLinearColorRGB(Color=&LerpLinearColors(StartColor=<rgba 1 1 1 1>, EndColor=<rgba 0 1.836 5 1>, LerpFactor=$User.StarDeath))
    ScaleA=true
    ScaleRGB=true
   )
   /SetVariables(Particles.Renderer2_Color:LinearColor=&LerpLinearColors(StartColor=<rgba 0.1063 0 1 1>, EndColor=<rgba 0 0.3672 1 2>, LerpFactor=$User.StarDeath)) {isSetParameters=true}
   /SetVariables(Particles.Renderer3_Color:LinearColor=<rgba 0.0042 0 0.0503 1>) {isSetParameters=true}
   &UpdateMeshOrientation(Rotation Rate=0.7, Rotation Vector=<0 0 1>)
   &FN_Light_Attributes(
    Color=&LerpLinearColors(StartColor=<rgba 1.394 0 5 1>, EndColor=<rgba 0 0.7887 3 1>, LerpFactor=0)
    Create Color Attribute=true
    Create Exponent Attribute=true
    Create Radius Attribute=true
    Radius=40
    Scale Color=&FloatFromCurve(FloatCurve=DataInterface() {propertyValues {Curve {keys [<at 0 1>,<at 0.1 0.65>,<at 0.2 1>,<at 0.3 1.1>,<at 0.4 0.65>,<at 0.5 1>,<at 0.6 0.8>,<at 0.7 1>,<at 0.8 1.1>,<at 0.9 0.75>,<at 1 1>]}}})
   )
   &ScaleMeshSize(Scale Factor=&VectorFromFloat(Value=&Lerp_Float(A=1, B=0.2, Alpha=$User.StarDeath)))
   /SetVariables(Particles.MeshRenderer2_Size:Vector3f=&Multiply_Vector(
     A=&VectorFromFloat(Value=0.05)
     B=&VectorFromFloat(Value=&Lerp_Float(A=1, B=0, Alpha=$User.StarDeath))
    )) {isSetParameters=true}
   /SetVariables(Particles.Renderer3_Size:Vector2f=&Vector2DFromFloat(Value=&Lerp_Float(A=0, B=25, Alpha=$User.StarDeath))) {isSetParameters=true}
   &ScaleSpriteSize(Scale Factor=&Vector2DFromFloat(Value=&Lerp_Float(A=1, B=0.5, Alpha=$User.StarDeath)))
  ]
  rendererClasses [
   &NiagaraMeshRendererProperties
   &NiagaraSpriteRendererProperties
   &NiagaraLightRendererProperties
  ]
  renderers [
   {rendererClass=&NiagaraMeshRendererProperties, rendererIndex=0}
   {rendererClass=&NiagaraMeshRendererProperties, rendererIndex=1}
   {rendererClass=&NiagaraSpriteRendererProperties, rendererIndex=2}
   {rendererClass=&NiagaraLightRendererProperties, rendererIndex=3}
   {rendererClass=&NiagaraSpriteRendererProperties, rendererIndex=4}
  ]
 }
}
summary {
 systemName=NS_LlamaRoaster_NeutronStar
 emitters [
  {
   emitterName=Sun
   rendererClasses [
    &NiagaraMeshRendererProperties
    &NiagaraSpriteRendererProperties
    &NiagaraLightRendererProperties
   ]
  }
  {emitterName=Planets, rendererClasses [&NiagaraMeshRendererProperties, &NiagaraSpriteRendererProperties]}
  {emitterName=SMoke, rendererClasses [&NiagaraSpriteRendererProperties]}
 ]
 userVariables [User.StarDeath {type=NiagaraFloat, defaultValue=0, description=""}]
}
