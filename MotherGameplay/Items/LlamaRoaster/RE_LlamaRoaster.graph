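# Reticle extension widget for the LlamaRoaster ranged weapon, showing a center dot and an overheat gauge.
# On construct it resets the gauge and binds to the equipped weapon's overheat event, which writes Current/Max as a FillPercent to five image materials.
# Targeting changes play the ADS animation forward or reverse, scaling and fading the gauge, or snap to its end state when a finished flag is set.
# ScaleRing is never called here.
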WidgetBlueprint /MotherGameplay/Items/LlamaRoaster/RE_LlamaRoaster
animations [
 ADS {
  binding_count=3
  length_seconds=0.1333
  object=&self:ADS
  sequence {
   bindings [
    OverlayContent {
     id=#1
     display_name=OverlayAll
     possessed_class=SizeBox
     tracks [
      {
       type=MovieScene2DTransformTrack
       sections [{
         type=MovieScene2DTransformSection
         channels [
          Translation.X {type=MovieSceneScriptingFloatChannel}
          Translation.Y {type=MovieSceneScriptingFloatChannel}
          Angle {type=MovieSceneScriptingFloatChannel}
          Scale.X {type=MovieSceneScriptingFloatChannel, keys [[0,0.666],[8,1]]}
          Scale.Y {type=MovieSceneScriptingFloatChannel, keys [[0,0.666],[8,1]]}
          Shear.X {type=MovieSceneScriptingFloatChannel}
          Shear.Y {type=MovieSceneScriptingFloatChannel}
         ]
        }]
      }
      {
       type=MovieSceneFloatTrack
       display_name=Render Opacity
       sections [{type=MovieSceneFloatSection, channels [None {type=MovieSceneScriptingFloatChannel, keys [[0,0.75],[8,1]]}]}]
      }
     ]
    }
    TopFade {
     id=#2
     possessed_class=Image
     tracks [
      {
       type=MovieSceneFloatTrack
       display_name=Render Opacity
       sections [{type=MovieSceneFloatSection, channels [None {type=MovieSceneScriptingFloatChannel, keys [[0,0],[8,1]]}]}]
      }
      {
       type=MovieSceneWidgetMaterialTrack
       sections [{type=MovieSceneParameterSection, channels [Fade {type=MovieSceneScriptingFloatChannel, keys [[0,4],[8,2]]}]}]
      }
     ]
    }
    BottomFade {
     id=#3
     possessed_class=Image
     tracks [
      {
       type=MovieSceneFloatTrack
       display_name=Render Opacity
       sections [{type=MovieSceneFloatSection, channels [None {type=MovieSceneScriptingFloatChannel, keys [[0,0],[8,1]]}]}]
      }
      {
       type=MovieSceneWidgetMaterialTrack
       sections [{type=MovieSceneParameterSection, channels [Fade {type=MovieSceneScriptingFloatChannel, keys [[0,4],[8,2]]}]}]
      }
     ]
    }
   ]
   display_rate {numerator=60}
  }
 }
 Intro {
  binding_count=2
  length_seconds=0.2
  object=&self:Intro
  sequence {
   bindings [
    OverlayAll {
     id=#4
     display_name=SizeBox_0
     possessed_class=Overlay
     tracks [
      {
       type=MovieScene2DTransformTrack
       sections [{
         type=MovieScene2DTransformSection
         channels [
          Translation.X {type=MovieSceneScriptingFloatChannel, const [0,1]}
          Translation.Y {type=MovieSceneScriptingFloatChannel, const [0,1]}
          Angle {type=MovieSceneScriptingFloatChannel}
          Scale.X {type=MovieSceneScriptingFloatChannel, keys [[0,0.75],[9,1]]}
          Scale.Y {type=MovieSceneScriptingFloatChannel, keys [[0,0.75],[9,1]]}
          Shear.X {type=MovieSceneScriptingFloatChannel, const [0,1]}
          Shear.Y {type=MovieSceneScriptingFloatChannel, const [0,1]}
         ]
        }]
      }
      {
       type=MovieSceneFloatTrack
       display_name=Render Opacity
       sections [{type=MovieSceneFloatSection, channels [None {type=MovieSceneScriptingFloatChannel, keys [[0,0],[9,1]]}]}]
      }
     ]
    }
    Clamp {
     id=#5
     possessed_class=Image
     tracks [{
       type=MovieSceneWidgetMaterialTrack
       sections [{type=MovieSceneParameterSection, channels [RingWidth {type=MovieSceneScriptingFloatChannel, keys [[-3,8],[0,8],[9.001,1.5]]}]}]
      }]
    }
   ]
   display_rate {numerator=60}
  }
 }
 ADS_EndState {
  binding_count=3
  length_seconds=0.1
  object=&self:ADS_EndState
  sequence {bindings [
    OverlayContent {
     id=#1
     display_name=OverlayAll
     possessed_class=SizeBox
     tracks [
      {
       type=MovieScene2DTransformTrack
       sections [{
         type=MovieScene2DTransformSection
         end_frame=2
         start_frame=0
         channels [
          Translation.X {type=MovieSceneScriptingFloatChannel}
          Translation.Y {type=MovieSceneScriptingFloatChannel}
          Angle {type=MovieSceneScriptingFloatChannel}
          Scale.X {type=MovieSceneScriptingFloatChannel, const [1,1]}
          Scale.Y {type=MovieSceneScriptingFloatChannel, const [1,1]}
          Shear.X {type=MovieSceneScriptingFloatChannel}
          Shear.Y {type=MovieSceneScriptingFloatChannel}
         ]
        }]
      }
      {
       type=MovieSceneFloatTrack
       display_name=Render Opacity
       sections [{type=MovieSceneFloatSection, channels [None {type=MovieSceneScriptingFloatChannel, const [1,1]}]}]
      }
     ]
    }
    TopFade {
     id=#2
     possessed_class=Image
     tracks [
      {
       type=MovieSceneFloatTrack
       display_name=Render Opacity
       sections [{type=MovieSceneFloatSection, channels [None {type=MovieSceneScriptingFloatChannel, const [1,1]}]}]
      }
      {
       type=MovieSceneWidgetMaterialTrack
       sections [{type=MovieSceneParameterSection, channels [Fade {type=MovieSceneScriptingFloatChannel, const [2,1]}]}]
      }
     ]
    }
    BottomFade {
     id=#3
     possessed_class=Image
     tracks [
      {
       type=MovieSceneFloatTrack
       display_name=Render Opacity
       sections [{type=MovieSceneFloatSection, end_frame=2, start_frame=0, channels [None {type=MovieSceneScriptingFloatChannel, const [1,1]}]}]
      }
      {
       type=MovieSceneWidgetMaterialTrack
       sections [{type=MovieSceneParameterSection, end_frame=2, start_frame=0, channels [Fade {type=MovieSceneScriptingFloatChannel, const [2,1]}]}]
      }
     ]
    }
   ]}
 }
 ADS_StartState {
  binding_count=3
  length_seconds=0.1
  object=&self:ADS_StartState
  sequence {bindings [
    OverlayContent {
     id=#1
     display_name=OverlayAll
     possessed_class=SizeBox
     tracks [
      {
       type=MovieScene2DTransformTrack
       sections [{
         type=MovieScene2DTransformSection
         end_frame=2
         start_frame=0
         channels [
          Translation.X {type=MovieSceneScriptingFloatChannel}
          Translation.Y {type=MovieSceneScriptingFloatChannel}
          Angle {type=MovieSceneScriptingFloatChannel}
          Scale.X {type=MovieSceneScriptingFloatChannel, const [0.666,1]}
          Scale.Y {type=MovieSceneScriptingFloatChannel, const [0.666,1]}
          Shear.X {type=MovieSceneScriptingFloatChannel}
          Shear.Y {type=MovieSceneScriptingFloatChannel}
         ]
        }]
      }
      {
       type=MovieSceneFloatTrack
       display_name=Render Opacity
       sections [{type=MovieSceneFloatSection, channels [None {type=MovieSceneScriptingFloatChannel, const [0.75,1]}]}]
      }
     ]
    }
    TopFade {
     id=#2
     possessed_class=Image
     tracks [
      {
       type=MovieSceneFloatTrack
       display_name=Render Opacity
       sections [{type=MovieSceneFloatSection, channels [None {type=MovieSceneScriptingFloatChannel, const [0,1]}]}]
      }
      {
       type=MovieSceneWidgetMaterialTrack
       sections [{type=MovieSceneParameterSection, channels [Fade {type=MovieSceneScriptingFloatChannel, const [4,1]}]}]
      }
     ]
    }
    BottomFade {
     id=#3
     possessed_class=Image
     tracks [
      {
       type=MovieSceneFloatTrack
       display_name=Render Opacity
       sections [{type=MovieSceneFloatSection, channels [None {type=MovieSceneScriptingFloatChannel, const [0,1]}]}]
      }
      {
       type=MovieSceneWidgetMaterialTrack
       sections [{type=MovieSceneParameterSection, channels [Fade {type=MovieSceneScriptingFloatChannel, const [4,1]}]}]
      }
     ]
    }
   ]}
 }
]
graphs [
 EventGraph {nodes [
   n1 = CustomEvent_2() {is=title K2Node_CustomEvent}
   > Set ADSAnimFinshedPlaying(ADSAnimFinshedPlaying:bool=false) -> (Output_Get:bool)
   n2 = CustomEvent_1() {is=title K2Node_CustomEvent}
   > Set ADSAnimFinshedPlaying(ADSAnimFinshedPlaying=true)
   n3 = CustomEvent_0!() -> (OutputDelegate, then, CurrentValue:float=~, MaxValue:float=~, State:EFortWeaponOverheatState Enum=None, bIsInCooldown:bool=~) {is=title K2Node_CustomEvent}
   > UpdateOverheatBar!(execute, self, Current:double=@n3.CurrentValue, Max:double=@n3.MaxValue) -> (then)
   Event Construct()
   > SetADSAnimStartState()
   > ResetOverHeatBar()
   > n4 = Cast To FortWeaponRanged!(
    execute
    Object=GetEquippedWeapon!(self:Fort Weapon Reticle Extension Widget Base Object Reference) -> (ReturnValue:Fort Weapon Object Reference)
   ) -> (then, CastFailed, AsFort Weapon Ranged:Fort Weapon Ranged Object Reference) {is=title K2Node_DynamicCast}
   > Bind Event to On Overheat Value Changed(self:Fort Weapon Ranged Object Reference=@n4, Delegate=@n3) {is=title K2Node_AddDelegate}
   n5 = Event OnTargetingChanged!() -> (OutputDelegate, then, bIsADS:bool) {is=title K2Node_Event}
   > n6 = Branch(Condition=@n5.bIsADS)
   > n7 = Branch(Condition=Get ADSAnimFinshedPlaying() -> (ADSAnimFinshedPlaying:bool))
   > SetADSAnimFinishedState()
   PlayAnimationForward(execute=@n7.else, InAnimation=Get ADS() -> (ADS:Widget Animation Object Reference))
   > BindToAnimationFinished(Animation=Get ADS(), Delegate=@n2)
   PlayAnimationReverse(execute=@n6.else, InAnimation=Get ADS())
   > BindToAnimationFinished(Animation=Get ADS(), Delegate=@n1)
   Event PreConstruct()
   Set ADSAnimFinshedPlaying(ADSAnimFinshedPlaying=true)
  ]}
 ResetOverHeatBar {nodes [
   n1 = GetDynamicMaterial(self=Get BottomFade() -> (BottomFade:Image Object Reference))
   n2 = Get ResetValue() -> (ResetValue:double)
   > SetScalarParameterValue(self=@n1, ParameterName=FillPercent, Value=@n2)
   > n3 = GetDynamicMaterial(self=Get TopFade() -> (TopFade:Image Object Reference))
   > SetScalarParameterValue(self=@n3, ParameterName=FillPercent, Value=@n2)
   ResetOverHeatBar() {is=title K2Node_FunctionEntry}
   > n4 = GetDynamicMaterial(self=Get Progress() -> (Progress:Image Object Reference))
   > SetScalarParameterValue(self=@n4, ParameterName=FillPercent, Value=@n2)
   > n5 = GetDynamicMaterial(self=Get Clamp() -> (Clamp:Image Object Reference))
   > SetScalarParameterValue(self=@n5, ParameterName=FillPercent, Value=@n2)
   > n6 = GetDynamicMaterial(self=Get ProgressNoise() -> (ProgressNoise:Image Object Reference))
   > SetScalarParameterValue(self=@n6, ParameterName=FillPercent, Value=@n2)
  ]}
 ScaleRing {nodes [
   n1 = ScaleRing!() -> (then, Current:double=~, Max:double=~) {is=title K2Node_FunctionEntry}
   n2 = Lerp(A=1, B=1.25, Alpha=float / float(A=@n1, B=@n1.Max) {is=title K2Node_CallFunction})
   > SetRenderScale!(execute, self:Widget Object Reference=Get Ring() -> (Ring:Image Object Reference), Scale:Vector 2D Structure=MakeVector2D(X=@n2, Y=@n2)) -> (then)
  ]}
 SetADSAnimFinishedState {nodes [
   SetADSAnimFinishedState() {is=title K2Node_FunctionEntry}
   > PlayAnimation(InAnimation=Get ADS_EndState() -> (ADS_EndState:Widget Animation Object Reference))
  ]}
 SetADSAnimStartState {nodes [
   SetADSAnimStartState() {is=title K2Node_FunctionEntry}
   > PlayAnimation(InAnimation=Get ADS_StartState() -> (ADS_StartState:Widget Animation Object Reference))
  ]}
 SetADSAnimationState {nodes [
   SetADSAnimationState() {is=title K2Node_FunctionEntry}
   > n1 = Branch(Condition=Get Is ADS() -> (Is ADS:bool))
   > Branch(Condition=Get ADSAnimFinshedPlaying())
   > SetADSAnimFinishedState() {is=title K2Node_CallFunction}
   SetADSAnimStartState(execute=@n1.else) {is=title K2Node_CallFunction}
  ]}
 UpdateOverheatBar {nodes [
   n1 = GetDynamicMaterial(self=Get BottomFade())
   n2 = UpdateOverheatBar!() -> (then, Current:double=~, Max:double=~) {is=title K2Node_FunctionEntry}
   n3 = float * float(A=float / float(A=@n2, B=@n2.Max), B=100)
   SetScalarParameterValue(execute=@n1.then, self=@n1, ParameterName=FillPercent, Value=@n3)
   > n4 = GetDynamicMaterial(self=Get TopFade())
   > SetScalarParameterValue(self=@n4, ParameterName=FillPercent, Value=@n3)
   n5 = GetDynamicMaterial(execute=@n8.then, self=Get ProgressNoise())
   > SetScalarParameterValue(self=@n5, ParameterName=FillPercent, Value=@n3)
   n6 = GetDynamicMaterial(execute=@n2.then, self=Get Progress())
   > SetScalarParameterValue(self=@n6, ParameterName=FillPercent, Value=@n3)
   > n7 = GetDynamicMaterial(self=Get Clamp())
   > n8 = SetScalarParameterValue(self=@n7, ParameterName=FillPercent, Value=@n3)
  ]}
]
variables [
 ADSAnimFinshedPlaying {type=boolean}
 ADSStartState {type=boolean}
 Current Value {type=number}
 Is ADS {type=boolean}
 Max Value {type=number}
 ResetValue {type=number}
 State {type=EFortWeaponOverheatState, default=None}
]
widgets {
 info {
  inheritedWidgetCount=1
  parentClass=/Script/MotherGameplayUI.FortLlamaRoasterReticleExtensionWidget
  rootWidgetClass=&SizeBox
  widgetCount=13
 }
 widgets [
  SizeBox_0 {
   widgetClassPath=&SizeBox
   widget=SizeBox_0
   properties {bOverride_HeightOverride=true, bOverride_WidthOverride=true, heightOverride=320, widthOverride=320}
  }
  OverlayAll {widgetClassPath=&Overlay, parent=SizeBox_0, widget=OverlayAll, slot {class=SizeBoxSlot}}
  OverlayContent {
   widgetClassPath=&Overlay
   parent=OverlayAll
   widget=OverlayContent
   properties {renderTransform {angle=0, scale=<0.666 0.666>, shear=<0 0>, translation=<0 0>}}
   slot {class=OverlaySlot, properties {horizontalAlignment=HAlign_Fill, verticalAlignment=VAlign_Fill}}
  }
  ProgressBacking {
   widgetClassPath=&Image
   bIsVariable=true
   parent=OverlayContent
   widget=ProgressBacking
   properties {
    visibility=SelfHitTestInvisible
    brush {
     bIsDynamicallyLoaded=false
     drawAs=Image
     imageType=NoImage
     mirroring=NoMirror
     resourceName=None
     resourceObject=/MotherGameplay/Items/LlamaRoaster/UI_Assets/M_UI_LlamaRoaster_Overheat_Clamp_Backing
     tiling=NoTile
     imageSize=<32 32>
     margin=<ltrb 0 0 0 0>
     outlineSettings {
      bUseBrushTransparency=false
      roundingType=HalfHeightRadius
      width=0
      color {colorUseRule=UseColor_Specified, specifiedColor=<rgba 0 0 0 0>}
      cornerRadii=<0 0 0 0>
     }
     tintColor {colorUseRule=UseColor_Specified, specifiedColor=<rgba 1 1 1 1>}
     uVRegion {bIsValid=false, max=<0 0>, min=<0 0>}
    }
   }
   slot {class=OverlaySlot, properties {horizontalAlignment=HAlign_Fill, verticalAlignment=VAlign_Fill}}
  }
  ProgressNoise {
   widgetClassPath=&Image
   bIsVariable=true
   parent=OverlayContent
   widget=ProgressNoise
   properties {
    visibility=SelfHitTestInvisible
    brush {
     bIsDynamicallyLoaded=false
     drawAs=Image
     imageType=NoImage
     mirroring=NoMirror
     resourceName=None
     resourceObject=/MotherGameplay/Items/LlamaRoaster/UI_Assets/M_UI_LlamaRoaster_Overheat_Filll_Inst
     tiling=NoTile
     imageSize=<32 32>
     margin=<ltrb 0 0 0 0>
     outlineSettings {
      bUseBrushTransparency=false
      roundingType=HalfHeightRadius
      width=0
      color {colorUseRule=UseColor_Specified, specifiedColor=<rgba 0 0 0 0>}
      cornerRadii=<0 0 0 0>
     }
     tintColor {colorUseRule=UseColor_Specified, specifiedColor=<rgba 1 1 1 1>}
     uVRegion {bIsValid=false, max=<0 0>, min=<0 0>}
    }
   }
   slot {class=OverlaySlot, properties {horizontalAlignment=HAlign_Fill, verticalAlignment=VAlign_Fill}}
  }
  Progress {
   widgetClassPath=&Image
   bIsVariable=true
   parent=OverlayContent
   widget=Progress
   properties {
    visibility=SelfHitTestInvisible
    brush {
     bIsDynamicallyLoaded=false
     drawAs=Image
     imageType=NoImage
     mirroring=NoMirror
     resourceName=None
     resourceObject=/MotherGameplay/Items/LlamaRoaster/UI_Assets/M_UI_LlamaRoaster_Overheat_Filll
     tiling=NoTile
     imageSize=<128 128>
     margin=<ltrb 0 0 0 0>
     outlineSettings {
      bUseBrushTransparency=false
      roundingType=HalfHeightRadius
      width=0
      color {colorUseRule=UseColor_Specified, specifiedColor=<rgba 0 0 0 0>}
      cornerRadii=<0 0 0 0>
     }
     tintColor {colorUseRule=UseColor_Specified, specifiedColor=<rgba 1 1 1 1>}
     uVRegion {bIsValid=false, max=<0 0>, min=<0 0>}
    }
   }
   slot {class=OverlaySlot, properties {horizontalAlignment=HAlign_Fill, verticalAlignment=VAlign_Fill}}
  }
  Clamp {
   widgetClassPath=&Image
   bIsVariable=true
   parent=OverlayContent
   widget=Clamp
   properties {
    visibility=SelfHitTestInvisible
    brush {
     bIsDynamicallyLoaded=false
     drawAs=Image
     imageType=NoImage
     mirroring=NoMirror
     resourceName=None
     resourceObject=/MotherGameplay/Items/LlamaRoaster/UI_Assets/M_UI_LlamaRoaster_Overheat_Bracket
     tiling=NoTile
     imageSize=<128 128>
     margin=<ltrb 0 0 0 0>
     outlineSettings {
      bUseBrushTransparency=false
      roundingType=HalfHeightRadius
      width=0
      color {colorUseRule=UseColor_Specified, specifiedColor=<rgba 0 0 0 0>}
      cornerRadii=<0 0 0 0>
     }
     tintColor {colorUseRule=UseColor_Specified, specifiedColor=<rgba 1 1 1 1>}
     uVRegion {bIsValid=false, max=<0 0>, min=<0 0>}
    }
   }
   slot {class=OverlaySlot, properties {horizontalAlignment=HAlign_Fill, verticalAlignment=VAlign_Fill}}
  }
  TopFade {
   widgetClassPath=&Image
   bIsVariable=true
   parent=OverlayContent
   widget=TopFade
   properties {
    renderOpacity=0
    visibility=SelfHitTestInvisible
    brush {
     bIsDynamicallyLoaded=false
     drawAs=Image
     imageType=NoImage
     mirroring=NoMirror
     resourceName=None
     resourceObject=/MotherGameplay/Items/LlamaRoaster/UI_Assets/M_UI_LlamaRoaster_Overheat_Bracket_Fade_1
     tiling=NoTile
     imageSize=<32 32>
     margin=<ltrb 0 0 0 0>
     outlineSettings {
      bUseBrushTransparency=false
      roundingType=HalfHeightRadius
      width=0
      color {colorUseRule=UseColor_Specified, specifiedColor=<rgba 0 0 0 0>}
      cornerRadii=<0 0 0 0>
     }
     tintColor {colorUseRule=UseColor_Specified, specifiedColor=<rgba 1 1 1 1>}
     uVRegion {bIsValid=false, max=<0 0>, min=<0 0>}
    }
    renderTransform {angle=-93, scale=<1 1>, shear=<0 0>, translation=<0 0>}
   }
   slot {class=OverlaySlot, properties {horizontalAlignment=HAlign_Fill, verticalAlignment=VAlign_Fill}}
  }
  BottomFade {
   widgetClassPath=&Image
   bIsVariable=true
   parent=OverlayContent
   widget=BottomFade
   properties {
    renderOpacity=0
    visibility=SelfHitTestInvisible
    brush {
     bIsDynamicallyLoaded=false
     drawAs=Image
     imageType=NoImage
     mirroring=NoMirror
     resourceName=None
     resourceObject=/MotherGameplay/Items/LlamaRoaster/UI_Assets/M_UI_LlamaRoaster_Overheat_Bracket_Fade_2
     tiling=NoTile
     imageSize=<32 32>
     margin=<ltrb 0 0 0 0>
     outlineSettings {
      bUseBrushTransparency=false
      roundingType=HalfHeightRadius
      width=0
      color {colorUseRule=UseColor_Specified, specifiedColor=<rgba 0 0 0 0>}
      cornerRadii=<0 0 0 0>
     }
     tintColor {colorUseRule=UseColor_Specified, specifiedColor=<rgba 1 1 1 1>}
     uVRegion {bIsValid=false, max=<0 0>, min=<0 0>}
    }
    renderTransform {angle=93, scale=<1 1>, shear=<0 0>, translation=<0 0>}
   }
   slot {class=OverlaySlot, properties {horizontalAlignment=HAlign_Fill, verticalAlignment=VAlign_Fill}}
  }
  Ring {
   widgetClassPath=&Image
   bIsVariable=true
   parent=OverlayContent
   widget=Ring
   properties {
    visibility=Collapsed
    brush {
     bIsDynamicallyLoaded=false
     drawAs=Image
     imageType=NoImage
     mirroring=NoMirror
     resourceName=None
     resourceObject=/MotherGameplay/Items/LlamaRoaster/UI_Assets/M_UI_LlamaRoaster_Reticle
     tiling=NoTile
     imageSize=<48 48>
     margin=<ltrb 0 0 0 0>
     outlineSettings {
      bUseBrushTransparency=false
      roundingType=HalfHeightRadius
      width=0
      color {colorUseRule=UseColor_Specified, specifiedColor=<rgba 0 0 0 0>}
      cornerRadii=<0 0 0 0>
     }
     tintColor {colorUseRule=UseColor_Specified, specifiedColor=<rgba 1 1 1 1>}
     uVRegion {bIsValid=false, max=<0 0>, min=<0 0>}
    }
   }
   slot {class=OverlaySlot, properties {horizontalAlignment=HAlign_Center, verticalAlignment=VAlign_Center}}
  }
  DotShadow {
   widgetClassPath=&Image
   bIsVariable=true
   parent=OverlayAll
   widget=DotShadow
   properties {
    visibility=SelfHitTestInvisible
    brush {
     bIsDynamicallyLoaded=false
     drawAs=Image
     imageType=NoImage
     mirroring=NoMirror
     resourceName=None
     resourceObject=/MotherGameplay/Items/LlamaRoaster/UI_Assets/M_UI_LlamaRoaster_Circle_Shadow
     tiling=NoTile
     imageSize=<10 10>
     margin=<ltrb 0 0 0 0>
     outlineSettings {
      bUseBrushTransparency=false
      roundingType=HalfHeightRadius
      width=0
      color {colorUseRule=UseColor_Specified, specifiedColor=<rgba 0 0 0 0>}
      cornerRadii=<0 0 0 0>
     }
     tintColor {colorUseRule=UseColor_Specified, specifiedColor=<rgba 1 1 1 1>}
     uVRegion {bIsValid=false, max=<0 0>, min=<0 0>}
    }
   }
   slot {class=OverlaySlot, properties {horizontalAlignment=HAlign_Center, verticalAlignment=VAlign_Center}}
  }
  Dot {
   widgetClassPath=&Image
   bIsVariable=true
   parent=OverlayAll
   widget=Dot
   properties {
    visibility=SelfHitTestInvisible
    brush {
     bIsDynamicallyLoaded=false
     drawAs=Image
     imageType=NoImage
     mirroring=NoMirror
     resourceName=None
     resourceObject=/MotherGameplay/Items/LlamaRoaster/UI_Assets/M_UI_LlamaRoaster_Circle
     tiling=NoTile
     imageSize=<8 8>
     margin=<ltrb 0 0 0 0>
     outlineSettings {
      bUseBrushTransparency=false
      roundingType=HalfHeightRadius
      width=0
      color {colorUseRule=UseColor_Specified, specifiedColor=<rgba 0 0 0 0>}
      cornerRadii=<0 0 0 0>
     }
     tintColor {colorUseRule=UseColor_Specified, specifiedColor=<rgba 1 1 1 1>}
     uVRegion {bIsValid=false, max=<0 0>, min=<0 0>}
    }
   }
   slot {class=OverlaySlot, properties {horizontalAlignment=HAlign_Center, verticalAlignment=VAlign_Center}}
  }
  TutorialHighlight {widgetClassPath=&TutorialHighlight}
 ]
}
